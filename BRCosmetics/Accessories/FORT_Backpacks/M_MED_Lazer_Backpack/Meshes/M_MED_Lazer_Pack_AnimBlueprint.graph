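# Animation Blueprint for the M_MED_Lazer backpack cosmetic, posing its wings from the owning character's movement and spinning the dyn_sun bone.
# WingPosition state machine selects pose-asset poses for base, targeting, crouch, jump, fall and nested skydive states, with an acceleration-driven additive blendspace on base.
# The cached result receives an additive sprint pose weighted by forward speed, zeroed during vertical motion.
# Event Graph computes these inputs and advances the wrapping spin rotator each update.
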AnimBlueprint /BRCosmetics/Accessories/FORT_Backpacks/M_MED_Lazer_Backpack/Meshes/M_MED_Lazer_Pack_AnimBlueprint
target_skeleton=/Game/Accessories/FORT_Backpacks/M_MED_Lazer_Backpack/Meshes/M_MED_Lazer_Pack_Skeleton
graphs [
 AnimGraph {nodes [
   n1 = Wings_Movement_AddPoses_PoseAsset : Sprint!(PoseWeight:float=Get SprintSpeedRemap() -> (SprintSpeedRemap:double)) -> (Pose:Pose Link Structure) {
    is=title AnimGraphNode_PoseByName
    poseAsset=/Game/Accessories/FORT_Backpacks/M_MED_Lazer_Backpack/Animation/Wings_Movement_AddPoses_PoseAsset
    poseName=Sprint
   }
   n2 = Use cached pose 'preSprint'()
   n3 = "Transform (Modify) Bone - Bone: dyn_sun"(
    ComponentPose=Local To Component(LocalPose=Apply Additive(Base=@n2, Additive=@n1, Alpha=0.5) {alphaBoolBlend {blendOption=Linear}})
    Alpha=1
    Translation="0,0,0"
    Rotation=Get ball_rotator() -> (ball_rotator:Rotator)
    Scale="1,1,1"
   ) {rotationMode=BMM_Additive, rotationSpace=BCS_BoneSpace, alphaBoolBlend {blendOption=Linear}, boneToModify {boneName=dyn_sun}}
   Output Pose(Result=Component To Local(ComponentPose=@n3)) {layerGroup=None}
   AnimGraphNode_StateMachine_0 = WingPosition() {is=title AnimGraphNode_StateMachine}
   Save cached pose 'preSprint'(Pose=@AnimGraphNode_StateMachine_0) {cacheName=preSprint}
  ]}
 WingPosition {
  entry=Base_
  path=AnimGraph.AnimGraphNode_StateMachine_0.WingPosition
  ids=id
  nodes [
   AnimStateNode_0 = Base_() {is=title AnimStateNode}
   AnimStateNode_1 = Targeting() {is=title AnimStateNode}
   AnimStateNode_2 = Crouching() {is=title AnimStateNode}
   AnimStateNode_3 = Falling() {is=title AnimStateNode}
   AnimStateNode_4 = Jump() {is=title AnimStateNode}
   AnimStateNode_5 = SkydiveMode() {is=title AnimStateNode}
   AnimStateTransitionNode_0 = AnimStateTransitionNode()
   AnimStateTransitionNode_1 = AnimStateTransitionNode()
   AnimStateTransitionNode_10 = AnimStateTransitionNode()
   AnimStateTransitionNode_11 = AnimStateTransitionNode()
   AnimStateTransitionNode_2 = AnimStateTransitionNode()
   AnimStateTransitionNode_3 = AnimStateTransitionNode()
   AnimStateTransitionNode_4 = AnimStateTransitionNode()
   AnimStateTransitionNode_5 = AnimStateTransitionNode()
   AnimStateTransitionNode_6 = AnimStateTransitionNode()
   AnimStateTransitionNode_7 = AnimStateTransitionNode()
   AnimStateTransitionNode_8 = AnimStateTransitionNode()
   AnimStateTransitionNode_9 = AnimStateTransitionNode()
  ]
  links [
   {
    dst=Crouching
    rule=AnimGraph.AnimGraphNode_StateMachine_0.WingPosition.AnimStateTransitionNode_1.AnimationTransitionGraph_0
    src=Base_
    via=AnimStateTransitionNode_1
   }
   {
    dst=Falling
    rule=AnimGraph.AnimGraphNode_StateMachine_0.WingPosition.AnimStateTransitionNode_3.AnimationTransitionGraph_0
    src=Base_
    via=AnimStateTransitionNode_3
   }
   {
    dst=Jump
    rule=AnimGraph.AnimGraphNode_StateMachine_0.WingPosition.AnimStateTransitionNode_7.AnimationTransitionGraph_0
    src=Base_
    via=AnimStateTransitionNode_7
   }
   {
    dst=SkydiveMode
    rule=AnimGraph.AnimGraphNode_StateMachine_0.WingPosition.AnimStateTransitionNode_9.AnimationTransitionGraph_0
    src=Base_
    via=AnimStateTransitionNode_9
   }
   {
    dst=Targeting
    rule=AnimGraph.AnimGraphNode_StateMachine_0.WingPosition.AnimStateTransitionNode_0.AnimationTransitionGraph_0
    src=Base_
    via=AnimStateTransitionNode_0
   }
   {
    dst=Base_
    rule=AnimGraph.AnimGraphNode_StateMachine_0.WingPosition.AnimStateTransitionNode_6.AnimationTransitionGraph_0
    src=Crouching
    via=AnimStateTransitionNode_6
   }
   {
    dst=Base_
    rule=AnimGraph.AnimGraphNode_StateMachine_0.WingPosition.AnimStateTransitionNode_4.AnimationTransitionGraph_0
    src=Falling
    via=AnimStateTransitionNode_4
   }
   {
    dst=Base_
    rule=AnimGraph.AnimGraphNode_StateMachine_0.WingPosition.AnimStateTransitionNode_8.AnimationTransitionGraph_0
    src=Jump
    via=AnimStateTransitionNode_8
   }
   {
    dst=Falling
    rule=AnimGraph.AnimGraphNode_StateMachine_0.WingPosition.AnimStateTransitionNode_2.AnimationTransitionGraph_0
    src=Jump
    via=AnimStateTransitionNode_2
   }
   {
    dst=SkydiveMode
    rule=AnimGraph.AnimGraphNode_StateMachine_0.WingPosition.AnimStateTransitionNode_11.AnimationTransitionGraph_0
    src=Jump
    via=AnimStateTransitionNode_11
   }
   {
    dst=Base_
    rule=AnimGraph.AnimGraphNode_StateMachine_0.WingPosition.AnimStateTransitionNode_10.AnimationTransitionGraph_0
    src=SkydiveMode
    via=AnimStateTransitionNode_10
   }
   {
    dst=Base_
    rule=AnimGraph.AnimGraphNode_StateMachine_0.WingPosition.AnimStateTransitionNode_5.AnimationTransitionGraph_0
    src=Targeting
    via=AnimStateTransitionNode_5
   }
  ]
 }
 Base_ {
  path=AnimGraph.AnimGraphNode_StateMachine_0.WingPosition.AnimStateNode_0.Base_
  nodes [
   n1 = Get PlayerVector_FwdBwd() -> (PlayerVector_FwdBwd:double)
   n2 = Get PlayerVector_LR() -> (PlayerVector_LR:double)
   Output Animation Pose(Result=Apply Additive(
     Base=Wings_Movement_Poses_PoseAsset : Base!(PoseWeight:float=1) -> (Pose:Pose Link Structure) {
      is=title AnimGraphNode_PoseByName
      poseAsset=/Game/Accessories/FORT_Backpacks/M_MED_Lazer_Backpack/Animation/Wings_Movement_Poses_PoseAsset
      poseName=Base
     }
     Additive=Blendspace Player 'Ground_Movement_BS'(X=@n2, Y=@n1) {blendSpace=/Game/Accessories/FORT_Backpacks/M_MED_Lazer_Backpack/Animation/Ground_Movement_BS}
    ) {alphaBoolBlend {blendOption=Linear}})
  ]
 }
 Targeting {
  path=AnimGraph.AnimGraphNode_StateMachine_0.WingPosition.AnimStateNode_1.Targeting
  nodes [Output Animation Pose(Result=Wings_Movement_Poses_PoseAsset : Targeting!(PoseWeight:float=1) -> (Pose:Pose Link Structure) {
     is=title AnimGraphNode_PoseByName
     poseAsset=&Wings_Movement_Poses_PoseAsset
     poseName=Targeting
    })]
 }
 Crouching {
  path=AnimGraph.AnimGraphNode_StateMachine_0.WingPosition.AnimStateNode_2.Crouching
  nodes [Output Animation Pose(Result=Wings_Movement_Poses_PoseAsset : Crouch!(PoseWeight:float=1) -> (Pose:Pose Link Structure) {
     is=title AnimGraphNode_PoseByName
     poseAsset=&Wings_Movement_Poses_PoseAsset
     poseName=Crouch
    })]
 }
 Falling {
  path=AnimGraph.AnimGraphNode_StateMachine_0.WingPosition.AnimStateNode_3.Falling
  nodes [Output Animation Pose(Result=Wings_Movement_Poses_PoseAsset : Jump!(PoseWeight:float=1) -> (Pose:Pose Link Structure) {
     is=title AnimGraphNode_PoseByName
     poseAsset=&Wings_Movement_Poses_PoseAsset
     poseName=Jump
    })]
 }
 Jump {
  path=AnimGraph.AnimGraphNode_StateMachine_0.WingPosition.AnimStateNode_4.Jump
  nodes [Output Animation Pose(Result=Wings_Movement_Poses_PoseAsset : Jump(PoseWeight=1) {poseAsset=&Wings_Movement_Poses_PoseAsset, poseName=Jump})]
 }
 SkydiveMode {
  path=AnimGraph.AnimGraphNode_StateMachine_0.WingPosition.AnimStateNode_5.SkydiveMode
  nodes [
   AnimGraphNode_StateMachine_0 = Skydiving State() {is=title AnimGraphNode_StateMachine}
   Output Animation Pose(Result=@AnimGraphNode_StateMachine_0)
  ]
 }
 Skydiving State {
  entry=Parachuting
  path=AnimGraph.AnimGraphNode_StateMachine_0.WingPosition.AnimStateNode_5.SkydiveMode.AnimGraphNode_StateMachine_0.Skydiving State
  ids=id
  nodes [
   AnimStateNode_0 = Parachuting() {is=title AnimStateNode}
   AnimStateNode_1 = Diving() {is=title AnimStateNode}
   AnimStateNode_2 = State_2() {is=title AnimStateNode}
   AnimStateTransitionNode_3 = AnimStateTransitionNode()
   AnimStateTransitionNode_4 = AnimStateTransitionNode()
   AnimStateTransitionNode_5 = AnimStateTransitionNode()
   AnimStateTransitionNode_6 = AnimStateTransitionNode()
   AnimStateTransitionNode_7 = AnimStateTransitionNode()
   AnimStateTransitionNode_8 = AnimStateTransitionNode()
  ]
  links [
   {
    dst=Parachuting
    rule=AnimGraph.AnimGraphNode_StateMachine_0.WingPosition.AnimStateNode_5.SkydiveMode.AnimGraphNode_StateMachine_0.Skydiving State.AnimStateTransitionNode_6.Transition
    src=Diving
    via=AnimStateTransitionNode_6
   }
   {
    dst=State_2
    rule=AnimGraph.AnimGraphNode_StateMachine_0.WingPosition.AnimStateNode_5.SkydiveMode.AnimGraphNode_StateMachine_0.Skydiving State.AnimStateTransitionNode_4.Transition
    src=Diving
    via=AnimStateTransitionNode_4
   }
   {
    dst=Diving
    rule=AnimGraph.AnimGraphNode_StateMachine_0.WingPosition.AnimStateNode_5.SkydiveMode.AnimGraphNode_StateMachine_0.Skydiving State.AnimStateTransitionNode_3.Transition
    src=Parachuting
    via=AnimStateTransitionNode_3
   }
   {
    dst=State_2
    rule=AnimGraph.AnimGraphNode_StateMachine_0.WingPosition.AnimStateNode_5.SkydiveMode.AnimGraphNode_StateMachine_0.Skydiving State.AnimStateTransitionNode_7.Transition
    src=Parachuting
    via=AnimStateTransitionNode_7
   }
   {
    dst=Diving
    rule=AnimGraph.AnimGraphNode_StateMachine_0.WingPosition.AnimStateNode_5.SkydiveMode.AnimGraphNode_StateMachine_0.Skydiving State.AnimStateTransitionNode_8.Transition
    src=State_2
    via=AnimStateTransitionNode_8
   }
   {
    dst=Parachuting
    rule=AnimGraph.AnimGraphNode_StateMachine_0.WingPosition.AnimStateNode_5.SkydiveMode.AnimGraphNode_StateMachine_0.Skydiving State.AnimStateTransitionNode_5.Transition
    src=State_2
    via=AnimStateTransitionNode_5
   }
  ]
 }
 Parachuting {
  path=AnimGraph.AnimGraphNode_StateMachine_0.WingPosition.AnimStateNode_5.SkydiveMode.AnimGraphNode_StateMachine_0.Skydiving State.AnimStateNode_0.Parachuting
  nodes [Output Animation Pose(Result=Wings_Movement_Poses_PoseAsset : Floating!(PoseWeight:float=0.5) -> (Pose:Pose Link Structure) {
     is=title AnimGraphNode_PoseByName
     poseAsset=&Wings_Movement_Poses_PoseAsset
     poseName=Floating
    })]
 }
 Diving {
  path=AnimGraph.AnimGraphNode_StateMachine_0.WingPosition.AnimStateNode_5.SkydiveMode.AnimGraphNode_StateMachine_0.Skydiving State.AnimStateNode_1.Diving
  nodes [Output Animation Pose(Result=Wings_Movement_Poses_PoseAsset : Dive!(PoseWeight:float=1) -> (Pose:Pose Link Structure) {
     is=title AnimGraphNode_PoseByName
     poseAsset=&Wings_Movement_Poses_PoseAsset
     poseName=Dive
    })]
 }
 State_2 {
  path=AnimGraph.AnimGraphNode_StateMachine_0.WingPosition.AnimStateNode_5.SkydiveMode.AnimGraphNode_StateMachine_0.Skydiving State.AnimStateNode_2.State_2
  nodes [Output Animation Pose(Result=Wings_Movement_Poses_PoseAsset : Floating(PoseWeight=1) {poseAsset=&Wings_Movement_Poses_PoseAsset, poseName=Floating})]
 }
 Transition {
  path=AnimGraph.AnimGraphNode_StateMachine_0.WingPosition.AnimStateNode_5.SkydiveMode.AnimGraphNode_StateMachine_0.Skydiving State.AnimStateTransitionNode_3.Transition
  nodes [Result(bCanEnterTransition=NOT Boolean(A=Get bIsParachuteOpened()))]
 }
 Transition {
  path=AnimGraph.AnimGraphNode_StateMachine_0.WingPosition.AnimStateNode_5.SkydiveMode.AnimGraphNode_StateMachine_0.Skydiving State.AnimStateTransitionNode_4.Transition
  nodes [Result(bCanEnterTransition=NOT Boolean(A=Get bIsSkydiveDiveMode()))]
 }
 Transition {
  path=AnimGraph.AnimGraphNode_StateMachine_0.WingPosition.AnimStateNode_5.SkydiveMode.AnimGraphNode_StateMachine_0.Skydiving State.AnimStateTransitionNode_5.Transition
  nodes [Result(bCanEnterTransition=Get bIsParachuteOpened())]
 }
 Transition {
  path=AnimGraph.AnimGraphNode_StateMachine_0.WingPosition.AnimStateNode_5.SkydiveMode.AnimGraphNode_StateMachine_0.Skydiving State.AnimStateTransitionNode_6.Transition
  nodes [Result(bCanEnterTransition=Get bIsParachuteOpened())]
 }
 Transition {
  path=AnimGraph.AnimGraphNode_StateMachine_0.WingPosition.AnimStateNode_5.SkydiveMode.AnimGraphNode_StateMachine_0.Skydiving State.AnimStateTransitionNode_7.Transition
  nodes [Result(bCanEnterTransition=NOT Boolean(A=Get bIsParachuteOpened()))]
 }
 Transition {
  path=AnimGraph.AnimGraphNode_StateMachine_0.WingPosition.AnimStateNode_5.SkydiveMode.AnimGraphNode_StateMachine_0.Skydiving State.AnimStateTransitionNode_8.Transition
  nodes [Result(bCanEnterTransition=Get bIsSkydiveDiveMode())]
 }
 AnimationTransitionGraph_0 {
  path=AnimGraph.AnimGraphNode_StateMachine_0.WingPosition.AnimStateTransitionNode_0.AnimationTransitionGraph_0
  nodes [Result(bCanEnterTransition=Get bIsTargeting())]
 }
 AnimationTransitionGraph_0 {
  path=AnimGraph.AnimGraphNode_StateMachine_0.WingPosition.AnimStateTransitionNode_1.AnimationTransitionGraph_0
  nodes [Result(bCanEnterTransition=Get bIsCrouching())]
 }
 AnimationTransitionGraph_0 {
  path=AnimGraph.AnimGraphNode_StateMachine_0.WingPosition.AnimStateTransitionNode_10.AnimationTransitionGraph_0
  nodes [Result(bCanEnterTransition=NOT Boolean(A=Get bIsSkyDiving()))]
 }
 AnimationTransitionGraph_0 {
  path=AnimGraph.AnimGraphNode_StateMachine_0.WingPosition.AnimStateTransitionNode_11.AnimationTransitionGraph_0
  nodes [Result(bCanEnterTransition=Get bIsSkyDiving())]
 }
 AnimationTransitionGraph_0 {
  path=AnimGraph.AnimGraphNode_StateMachine_0.WingPosition.AnimStateTransitionNode_2.AnimationTransitionGraph_0
  nodes [Result(bCanEnterTransition=Get bIsFalling())]
 }
 AnimationTransitionGraph_0 {
  path=AnimGraph.AnimGraphNode_StateMachine_0.WingPosition.AnimStateTransitionNode_3.AnimationTransitionGraph_0
  nodes [Result(bCanEnterTransition=Get bIsFalling())]
 }
 AnimationTransitionGraph_0 {
  path=AnimGraph.AnimGraphNode_StateMachine_0.WingPosition.AnimStateTransitionNode_4.AnimationTransitionGraph_0
  nodes [Result(bCanEnterTransition=NOT Boolean(A=Get bIsFalling()))]
 }
 AnimationTransitionGraph_0 {
  path=AnimGraph.AnimGraphNode_StateMachine_0.WingPosition.AnimStateTransitionNode_5.AnimationTransitionGraph_0
  nodes [Result(bCanEnterTransition=NOT Boolean(A=Get bIsTargeting()))]
 }
 AnimationTransitionGraph_0 {
  path=AnimGraph.AnimGraphNode_StateMachine_0.WingPosition.AnimStateTransitionNode_6.AnimationTransitionGraph_0
  nodes [Result(bCanEnterTransition=NOT Boolean(A=Get bIsCrouching()))]
 }
 AnimationTransitionGraph_0 {
  path=AnimGraph.AnimGraphNode_StateMachine_0.WingPosition.AnimStateTransitionNode_7.AnimationTransitionGraph_0
  nodes [Result(bCanEnterTransition=Get bIsJumping())]
 }
 AnimationTransitionGraph_0 {
  path=AnimGraph.AnimGraphNode_StateMachine_0.WingPosition.AnimStateTransitionNode_8.AnimationTransitionGraph_0
  nodes [Result(bCanEnterTransition=NOT Boolean(A=Get bIsJumping()))]
 }
 AnimationTransitionGraph_0 {
  path=AnimGraph.AnimGraphNode_StateMachine_0.WingPosition.AnimStateTransitionNode_9.AnimationTransitionGraph_0
  nodes [Result(bCanEnterTransition=Get bIsSkyDiving())]
 }
 EventGraph {nodes [
   Event BlueprintUpdateAnimation()
   > n1 = Sequence!(execute) -> (then_0, then_1, then_2:Exec, then_3:Exec)
   n2 = Get Spin_Int() -> (Spin_Int:double)
   > Set Spin_Int(Spin_Int:double=float + float(A=@n2, B=1) {is=title K2Node_CommutativeAssociativeBinaryOperator}) -> (Output_Get:double)
   > n3 = Branch(Condition="float >= float"(A=@n2, B=720) {is=title K2Node_CallFunction})
   > n4 = Set Spin_Int(Spin_Int=float - float(A=Get Spin_Int(), B=720) {is=title K2Node_CallFunction})
   n5 = Get Spin_Delta() -> (Spin_Delta:double)
   n6 = Get SpinPerSecondFast() -> (SpinPerSecondFast:double)
   n7 = Set Spin_Int(execute=@n3.else, Spin_Int=float + float(A=Get Spin_Int(), B=float * float(A=@n5, B=@n6)))
   Set ball_rotator(execute=[@n4.then, @n7.then], ball_rotator:Rotator=MakeRotator(Roll=@n7, Pitch=@n7, Yaw=@n7)) -> (Output_Get:Rotator)
   n8 = TryGetPawnOwner()
   n9 = Cast To Character!(execute=@n1.then_1, Object=@n8) -> (then, CastFailed, AsCharacter:Character Object Reference) {is=title K2Node_DynamicCast}
   n10 = Get Actor Rotation!(self:Actor Object Reference=@n8) -> (ReturnValue:Rotator) {is=title K2Node_CallFunction}
   n11 = GetCurrentAcceleration!(self:Character Movement Component Object Reference=Get CharacterMovement!(self:Character Object Reference=@n9) -> (CharacterMovement:Character Movement Component Object Reference)) -> (ReturnValue:Vector)
   n12 = BreakVector(InVec=vector * vector(A:Vector=Get Rotation X Vector!(InRot:Rotator=@n10) -> (ReturnValue:Vector) {is=title K2Node_CallFunction}, B:Vector=@n11) -> (ReturnValue:Vector) {is=title K2Node_CommutativeAssociativeBinaryOperator})
   > Set PlayerVector_FwdBwd(PlayerVector_FwdBwd:double=float + float(A=@n12, B=@n12.Y)) -> (Output_Get:double)
   n13 = BreakVector(InVec=vector * vector(A=GetAxes!(A:Rotator=@n10) -> (X:Vector, Y:Vector, Z:Vector).Y, B=@n11))
   > Set PlayerVector_LR(PlayerVector_LR:double=float + float(A=@n13, B=@n13.Y)) -> (Output_Get:double)
   n14 = TryGetPawnOwner()
   n15 = GetVelocity!(self:Actor Object Reference=@n14) -> (ReturnValue:Vector)
   n16 = "Not Equal (Float)"!(A:double=BreakVector(InVec=@n15).Z, B:double=0) -> (ReturnValue:bool) {is=title K2Node_CallFunction}
   n17 = Get Actor Rotation(self=@n14)
   n18 = BreakVector(InVec=vector * vector(A=Get Rotation X Vector(InRot=@n17), B=@n15))
   n19 = float + float(A=@n18, B=@n18.Y)
   > Set SprintSpeedRemap(SprintSpeedRemap:double=float * float(A=MapRangeClamped(Value=@n19, InRangeA=0, InRangeB=550), B=SelectFloat(A=0, B=1, bPickA=@n16))) -> (Output_Get:double)
  ]}
]
variables [
 ball_rotator {type=Rotator}
 PlayerVector_FwdBwd {type=number}
 PlayerVector_LR {type=number}
 Spin_Delta {type=number}
 Spin_Int {type=number}
 SpinPerSecondFast {type=number}
 SprintSpeedRemap {type=number}
]
